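; Blueprint for the inflatable blimp celebration prop, a building actor in the Artemis environment set.
; Its construction script writes a random float between 0 and 1 into custom primitive data slot 0 on the static mesh component, giving each instance its own value for the material to read.
; BeginPlay is empty, so the graphs add no runtime behavior.
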(event EventBeginPlay)

(fn ConstructionScript
  (Rendering|Material|SetCustomPrimitiveDataFloat (Variables|Debug|GetStaticMeshComponent) 0 (Math|Random|RandomFloatinRange 0.0 1.0)))
