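; Celebration inflatable crash pad building actor that culls or repositions itself at match start.
; Once playlist data is ready, the server gives pads disabled by a data row a short random lifespan.
; If ocean rise, in 384-unit levels, is at or below the water level limit, the pad silently dies after a random delay that staggers the cost across frames.
; Otherwise it optionally snaps its Z to ocean height, then runs parent BeginPlay.
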(event EventBeginPlay
  (bind (_returnvalue_x _returnvalue_y) (Math|Vector|BreakVector (Transformation|GetActorLocation)))
  (bind _returnvalue (WorldSubsystems|GetWaterSubsystem))
  (bind _returnvalue_2 (Water|GetOceanTotalHeight _returnvalue))
  ;[1 1 1] Check if the env crash pad is enabled if not clean it off the map.
  (Class|FortAsyncActionCurrentPlaylistReady|PlaylistDataReadyAsync
    (:OnReady
      (if (Networking|IsServer)
        (if (not (Class|AthenaFunctionLibrary|RowtoBool (Variables|Default|GetRow_Enabled)))
          (Actor|SetLifeSpan (Math|Random|RandomFloatinRange 0.1 3.0)))
        (else
          _))
      ;[1 1 1] Set the world height for the Z of the pad if it is set to use ocean height. If the level is at or bellow bellow the water level int value then kill the pad.
      (if (>= (Math|Conversions|ToFloat(Integer) (Variables|Default|GetWaterLevelLimit)) (/ (- _returnvalue_2 (Water|GetOceanBaseHeight _returnvalue)) 384.0))
        (Utilities|FlowControl|Delay (Math|Random|RandomFloatinRange 0.01 1.0)) ; Kill on different frames for perf
        (Gameplay|SilentDie false)
        (else
          (if (Variables|Default|GetSetsZToOceanHeight)
            (Transformation|SetActorLocation 0 (Math|Vector|MakeVector _returnvalue_x _returnvalue_y _returnvalue_2) false false)
            (else
              _))
          (|CallParentFunction))))))

(fn ConstructionScript
  (|CallParentFunction))
